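# Physics asset for the female medium Commando Viking character, named as the collider for the cape, likely for cloth or secondary-motion simulation.
# Capsules approximate the head, clavicles, spine, pelvis and feet, with spine_02 to spine_04 each layering a second, differently oriented capsule.
# An extra capsule sits on loincloth_back04, while neck, arm, thigh and calf bodies have no shapes.
# It has no constraints, so it defines collision volumes without ragdoll joints.
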PhysicsAsset /Game/Characters/Player/Female/Medium/Bodies/F_MED_Commando_Viking_01/Meshes/F_MED_Commando_Viking_CapeCollider
bodies {
 calf_l []
 calf_r []
 clavicle_l [{shapeType=Capsule, length=6.1964, radius=10.1353, center=<7.9211 -0.5255 -0.7824>, rotation=<rot -77.5566 177.3915 7.7314>}]
 clavicle_r [{shapeType=Capsule, length=6.6964, radius=10.2957, center=<-6.9199 0.594 1.4051>, rotation=<rot -77.5565 177.3915 -172.2685>}]
 foot_l [{shapeType=Capsule, length=12.5982, radius=6.5127, center=<-2.2072 6.0534 -0.3367>, rotation=<rot 0.0388 1.4701 -89.3982>}]
 foot_r [{shapeType=Capsule, length=12.5982, radius=6.5127, center=<2.2072 -6.0534 0.3367>, rotation=<rot 0.0388 1.4701 90.6018>}]
 head [{shapeType=Capsule, length=0, radius=18, center=<9.0697 0.0746 0>, rotation=<rot 0 -90 0>}]
 loincloth_back04 [{shapeType=Capsule, length=10, radius=12.75, shapeName=loincloth_back04_capsule, center=<-1.447 -6.5849 0>, rotation=<rot 90 0 0.0002>}]
 lowerarm_l []
 lowerarm_r []
 neck_02 []
 pelvis [{shapeType=Capsule, length=9.9028, radius=13.6596, center=<-1.9544 -3.8035 -0.1981>}]
 spine_01 [{shapeType=Capsule, length=9.0171, radius=9.7354, center=<-2.8183 -1.0873 0>}]
 spine_02 [
  {shapeType=Capsule, length=5.5484, radius=11.2354, center=<1.179 0.2594 0>}
  {shapeType=Capsule, length=20.6875, radius=9.3125, center=<-1.8317 2.606 0>, rotation=<rot 90 0 0>}
 ]
 spine_03 [
  {shapeType=Capsule, length=5.5484, radius=11.7354, center=<2.1774 -1.3893 0>}
  {shapeType=Capsule, length=20.7188, radius=9.2812, center=<-1.2326 2.6059 0>, rotation=<rot 90 0 0>}
 ]
 spine_04 [
  {shapeType=Capsule, length=2.8296, radius=13.3916, center=<5.5182 0.56 0>}
  {shapeType=Capsule, length=10.0312, radius=9.2812, center=<-1.2326 1.308 0>, rotation=<rot 90 0 0>}
 ]
 spine_05 [{shapeType=Capsule, length=7.9859, radius=13.0791, center=<7.2978 0.3227 0>}]
 thigh_l []
 thigh_r []
 upperarm_l []
 upperarm_r []
}
constraints []
