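# Physics asset for the loincloth on the female medium Viking Commando body, covering both front and back despite the Front name.
# Capsules on the pelvis, thighs and loincloth bones form a front flap chain and a four-bone back chain, both rooted at the pelvis.
# Joints within each chain have tight twist and limited swing, with side swing widening down the back chain.
# Pelvis joints and extra thigh-to-upper-front-cloth constraints are angularly free.
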PhysicsAsset /Game/Characters/Player/Female/Medium/Bodies/F_MED_Commando_Viking_01/Meshes/F_MED_Commando_Viking_Cloth_Front_Physics
bodies {
 loincloth_back01 [{shapeType=Capsule, length=14.2157, radius=3.0987, center=<4.3524 -1.0807 1.3711>, rotation=<rot -3.6703 -90 -4.01>}]
 loincloth_back02 [{shapeType=Capsule, length=14.25, radius=7.5, center=<1.3729 5 0>}]
 loincloth_back03 [{shapeType=Capsule, length=11.25, radius=5.75, center=<4.1404 3.5 0>}]
 loincloth_back04 [{shapeType=Capsule, length=3.25, radius=5.5, center=<4.283 3 0>}]
 loincloth_frnt_01_c [{shapeType=Capsule, length=15.5, radius=8.75, center=<5 -8 0>}]
 loincloth_frnt_02_c [{shapeType=Capsule, length=13.25, radius=8.25, center=<5 -6.5 0>}]
 loincloth_frnt_03_c [{shapeType=Capsule, length=4.25, radius=8.25, center=<5 -6 0>}]
 loincloth_frnt_flap_c [{shapeType=Capsule, length=6.8425, radius=5.4684, center=<12.8364 0.2155 -0.5537>, rotation=<rot -0.1837 -90 1.9709>}]
 pelvis [{shapeType=Capsule, length=11.25, radius=12.25, center=<-3.3997 -0.2145 0.0003>}]
 thigh_l [{shapeType=Capsule, length=27.4474, radius=8.4852, center=<-19.7262 0.7556 -0.6789>, rotation=<rot 0.4367 -92.8715 -88.16>}]
 thigh_r [{shapeType=Capsule, length=27.135, radius=8.533, center=<20.0299 -0.6857 0.1371>, rotation=<rot 0.5813 -93.1571 -87.6868>}]
}
constraints [
 {bone1Name=loincloth_back02, bone2Name=loincloth_back01, swing1LimitDegrees=45, swing2LimitDegrees=15, twistLimitDegrees=3}
 {bone1Name=loincloth_back03, bone2Name=loincloth_back02, swing1LimitDegrees=45, swing2LimitDegrees=35, twistLimitDegrees=3}
 {bone1Name=loincloth_back04, bone2Name=loincloth_back03, swing1LimitDegrees=45, swing2LimitDegrees=45, twistLimitDegrees=3}
 {
  bone1Name=loincloth_back01
  bone2Name=pelvis
  swing1LimitDegrees=45
  swing1Motion=Free
  swing2LimitDegrees=45
  swing2Motion=Free
  twistLimitDegrees=45
  twistMotion=Free
 }
 {
  bone1Name=thigh_l
  bone2Name=pelvis
  swing1LimitDegrees=45
  swing1Motion=Free
  swing2LimitDegrees=45
  swing2Motion=Free
  twistLimitDegrees=45
  twistMotion=Free
 }
 {
  bone1Name=thigh_r
  bone2Name=pelvis
  swing1LimitDegrees=45
  swing1Motion=Free
  swing2LimitDegrees=45
  swing2Motion=Free
  twistLimitDegrees=45
  twistMotion=Free
 }
 {
  bone1Name=loincloth_frnt_flap_c
  bone2Name=pelvis
  swing1LimitDegrees=45
  swing1Motion=Free
  swing2LimitDegrees=45
  swing2Motion=Free
  twistLimitDegrees=45
  twistMotion=Free
 }
 {bone1Name=loincloth_frnt_01_c, bone2Name=loincloth_frnt_flap_c, swing1LimitDegrees=45, swing2LimitDegrees=20, twistLimitDegrees=3}
 {bone1Name=loincloth_frnt_02_c, bone2Name=loincloth_frnt_01_c, swing1LimitDegrees=45, swing2LimitDegrees=20, twistLimitDegrees=3}
 {bone1Name=loincloth_frnt_03_c, bone2Name=loincloth_frnt_02_c, swing1LimitDegrees=45, swing2LimitDegrees=20, twistLimitDegrees=3}
 {
  bone1Name=thigh_l
  bone2Name=loincloth_frnt_01_c
  swing1LimitDegrees=45
  swing1Motion=Free
  swing2LimitDegrees=45
  swing2Motion=Free
  twistLimitDegrees=45
  twistMotion=Free
 }
 {
  bone1Name=thigh_r
  bone2Name=loincloth_frnt_01_c
  swing1LimitDegrees=45
  swing1Motion=Free
  swing2LimitDegrees=45
  swing2Motion=Free
  twistLimitDegrees=45
  twistMotion=Free
 }
 {
  bone1Name=thigh_r
  bone2Name=loincloth_frnt_02_c
  swing1LimitDegrees=45
  swing1Motion=Free
  swing2LimitDegrees=45
  swing2Motion=Free
  twistLimitDegrees=45
  twistMotion=Free
 }
 {
  bone1Name=thigh_l
  bone2Name=loincloth_frnt_02_c
  swing1LimitDegrees=45
  swing1Motion=Free
  swing2LimitDegrees=45
  swing2Motion=Free
  twistLimitDegrees=45
  twistMotion=Free
 }
]
