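# Anim Blueprint for the female Viking Commando body, copying its parent mesh's pose and adding secondary motion.
# A Control Rig feeds serial AnimDynamics nodes for loincloth, hair, metal rings, earrings and shoulder fluff, plus ring constraints and an earring bone copy.
# Recent firing applies seeded random forces to hair, earrings and fluff.
# When the playlist uses custom character parts, a shrunken-head pose replaces all that processing.
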AnimBlueprint /Game/Characters/Player/Female/Medium/Bodies/F_MED_Commando_Viking_01/Meshes/F_MED_Commando_Viking_Skeleton_ABP
target_skeleton=&/F_MED_Commando_Viking_Skeleton
graphs [
 AnimGraph {nodes [
   n1 = "Transform (Modify) Bone - Bone: head"!(ComponentPose=Local To Component(LocalPose=Use cached pose 'PreScale Cache'()), Alpha=1) -> (Pose) {scaleMode=BMM_Replace, alphaBoolBlend {blendOption=Linear}, boneToModify {boneName=head}, scale=<0.01 0.01 0.01>}
   n2 = Get earrings_and_hair_external_force() -> (earrings_and_hair_external_force:Vector)
   n3 = Control Rig(Source=Use cached pose 'PreScale Cache'()) {
    alphaBoolBlend {blendOption=Linear}
    controlRigAssetReference {blueprintRigClass=&/F_MED_Commando_Viking_CtrlRig.F_MED_Commando_Viking_CtrlRig_C}
   }
   n4 = Anim Dynamics - Bone: loincloth_back01!(SimulationSpace:AnimPhysSimSpaceType Enum, ComponentPose=Local To Component(LocalPose=@n3), Alpha=1, GravityOverride:Vector="0,0,-980") -> (Pose) {
    angularDampingOverride=0.999
    angularSpringConstant=2000
    bAngularSpring=true
    bOverrideAngularDamping=true
    bUseGravityOverride=true
    bUseSphericalLimits=true
    gravityScale=0.35
    simulationSpace=RootRelative
    alphaBoolBlend {blendOption=Linear}
    boundBone {boneName=loincloth_back01}
    componentAppliedLinearAccClamp=<4000 4000 4000>
    componentLinearAccScale=<1 1 1>
    componentLinearVelScale=<1 1 1>
    gravityOverride {z=-980}
    physicsBodyDefinitions [{
      boundBone {boneName=loincloth_back01}
      boxExtents=<17 17 17>
      constraintSetup {angularLimitsMax=<0 0 45>, angularLimitsMin=<0 0 2>, angularTarget=<1 0 0>, linearAxesMax=<0 1 0>}
      localJointOffset=<2 0 0>
     }]
   }
   n5 = Anim Dynamics - Bone: loincloth_back02!(SimulationSpace:AnimPhysSimSpaceType Enum, ComponentPose=@n4, Alpha=1, GravityOverride:Vector="0,0,-980") -> (Pose) {
    angularDampingOverride=0.999
    angularSpringConstant=2000
    bAngularSpring=true
    bOverrideAngularDamping=true
    bUseGravityOverride=true
    gravityScale=0.35
    simulationSpace=RootRelative
    alphaBoolBlend {blendOption=Linear}
    boundBone {boneName=loincloth_back02}
    componentAppliedLinearAccClamp=<4000 4000 4000>
    componentLinearAccScale=<1 1 1>
    componentLinearVelScale=<1 1 1>
    gravityOverride {z=-980}
    physicsBodyDefinitions [{
      collisionType=CustomSphere
      boundBone {boneName=loincloth_back02}
      boxExtents=<17 17 17>
      constraintSetup {angularLimitsMax=<0 0 40>, angularLimitsMin=<0 0 -5>, angularTarget=<1 0 0>}
      localJointOffset=<2 0 0>
     }]
    sphericalLimits [
     {limitRadius=4, drivingBone {boneName=thigh_l}, sphereLocalOffset=<-5 0 0>}
     {limitRadius=4, drivingBone {boneName=thigh_r}, sphereLocalOffset=<5 0 0>}
    ]
   }
   n6 = Anim Dynamics - Bone: loincloth_back03!(SimulationSpace:AnimPhysSimSpaceType Enum, ComponentPose=@n5, Alpha=1, GravityOverride:Vector="0,0,-980") -> (Pose) {
    angularDampingOverride=0.999
    angularSpringConstant=2000
    bAngularSpring=true
    bOverrideAngularDamping=true
    bUseGravityOverride=true
    bUseSphericalLimits=true
    gravityScale=0.35
    simulationSpace=RootRelative
    alphaBoolBlend {blendOption=Linear}
    boundBone {boneName=loincloth_back03}
    componentAppliedLinearAccClamp=<4000 4000 4000>
    componentLinearAccScale=<1 1 1>
    componentLinearVelScale=<1 1 1>
    gravityOverride {z=-980}
    physicsBodyDefinitions [{
      collisionType=CustomSphere
      sphereCollisionRadius=11
      boundBone {boneName=loincloth_back03}
      boxExtents=<17 17 17>
      constraintSetup {angularLimitsMax=<0 5 40>, angularLimitsMin=<0 -5 -20>, angularTarget=<1 0 0>}
      localJointOffset=<2 0 0>
     }]
    sphericalLimits [
     {limitRadius=7, drivingBone {boneName=thigh_l}, sphereLocalOffset=<0 0 0>}
     {limitRadius=7, drivingBone {boneName=thigh_r}, sphereLocalOffset=<0 0 0>}
     {limitRadius=7, drivingBone {boneName=thigh_l}, sphereLocalOffset=<-15 3 0>}
     {limitRadius=7, drivingBone {boneName=thigh_r}, sphereLocalOffset=<15 -3 0>}
    ]
   }
   n7 = Anim Dynamics - Bone: loincloth_back04!(SimulationSpace:AnimPhysSimSpaceType Enum, ComponentPose=@n6, Alpha=1, GravityOverride:Vector="0,0,-980") -> (Pose) {
    angularDampingOverride=0.999
    angularSpringConstant=2000
    bAngularSpring=true
    bOverrideAngularDamping=true
    bUseGravityOverride=true
    gravityScale=0.35
    simulationSpace=RootRelative
    alphaBoolBlend {blendOption=Linear}
    boundBone {boneName=loincloth_back04}
    componentAppliedLinearAccClamp=<4000 4000 4000>
    componentLinearAccScale=<1 1 1>
    componentLinearVelScale=<1 1 1>
    gravityOverride {z=-980}
    physicsBodyDefinitions [{
      boundBone {boneName=loincloth_back04}
      boxExtents=<19 19 19>
      constraintSetup {angularLimitsMax=<10 15 30>, angularLimitsMin=<-10 -15 -30>, angularTarget=<1 0 0>}
      localJointOffset=<2 0 0>
     }]
   }
   n8 = "Anim Dynamics - Chain: loincloth_frnt_flap_l -> loincloth_frnt_01_l"!(SimulationSpace:AnimPhysSimSpaceType Enum, ComponentPose=@n7, Alpha=1, GravityOverride:Vector="0,0,0") -> (Pose) {
    angularDampingOverride=1
    angularSpringConstant=300
    bAngularSpring=true
    bChain=true
    bGravityOverrideInSimSpace=true
    bOverrideAngularDamping=true
    bUseGravityOverride=true
    numSolverIterationsPostUpdate=4
    numSolverIterationsPreUpdate=16
    simulationSpace=RootRelative
    alphaBoolBlend {blendOption=Linear}
    boundBone {boneName=loincloth_frnt_flap_l}
    chainEnd {boneName=loincloth_frnt_01_l}
    componentAppliedLinearAccClamp=<6000 6000 6000>
    componentLinearAccScale=<1 1 1>
    componentLinearVelScale=<1 1 1>
    physicsBodyDefinitions [
     {
      sphereCollisionRadius=2.5
      boundBone {boneName=loincloth_frnt_flap_l}
      boxExtents=<25 25 25>
      constraintSetup {angularLimitsMax=<0 0 0>, angularLimitsMin=<0 0 -25>, angularTarget=<0 0 0>}
      localJointOffset=<0 0 0>
     }
     {
      sphereCollisionRadius=2.5
      boundBone {boneName=loincloth_frnt_01_l}
      boxExtents=<25 25 25>
      constraintSetup {angularLimitsMax=<0 0 0>, angularLimitsMin=<0 0 -25>, angularTarget=<0 0 0>}
      localJointOffset=<0 0 0>
     }
    ]
   }
   n9 = "Anim Dynamics - Chain: loincloth_frnt_flap_r -> loincloth_frnt_01_r"!(SimulationSpace:AnimPhysSimSpaceType Enum, ComponentPose=@n8, Alpha=1, GravityOverride:Vector="0,0,0") -> (Pose) {
    angularDampingOverride=1
    angularSpringConstant=300
    bAngularSpring=true
    bChain=true
    bGravityOverrideInSimSpace=true
    bOverrideAngularDamping=true
    bUseGravityOverride=true
    numSolverIterationsPostUpdate=4
    numSolverIterationsPreUpdate=16
    simulationSpace=RootRelative
    alphaBoolBlend {blendOption=Linear}
    boundBone {boneName=loincloth_frnt_flap_r}
    chainEnd {boneName=loincloth_frnt_01_r}
    componentAppliedLinearAccClamp=<6000 6000 6000>
    componentLinearAccScale=<1 1 1>
    componentLinearVelScale=<1 1 1>
    physicsBodyDefinitions [
     {
      sphereCollisionRadius=2.5
      boundBone {boneName=loincloth_frnt_flap_r}
      boxExtents=<25 25 25>
      constraintSetup {angularLimitsMax=<0 0 0>, angularLimitsMin=<0 0 -25>, angularTarget=<0 0 0>}
      localJointOffset=<0 0 0>
     }
     {
      sphereCollisionRadius=2.5
      boundBone {boneName=loincloth_frnt_01_r}
      boxExtents=<25 25 25>
      constraintSetup {angularLimitsMax=<0 0 0>, angularLimitsMin=<0 0 -25>, angularTarget=<0 0 0>}
      localJointOffset=<0 0 0>
     }
    ]
   }
   n10 = "Anim Dynamics - Chain: loincloth_frnt_01_c -> loincloth_frnt_03_c"!(SimulationSpace:AnimPhysSimSpaceType Enum, ComponentPose=@n9, Alpha=1, GravityOverride:Vector="0,0,0") -> (Pose) {
    angularDampingOverride=1
    angularSpringConstant=1000
    bAngularSpring=true
    bChain=true
    bGravityOverrideInSimSpace=true
    bOverrideAngularDamping=true
    bUseGravityOverride=true
    numSolverIterationsPostUpdate=4
    numSolverIterationsPreUpdate=16
    simulationSpace=RootRelative
    alphaBoolBlend {blendOption=Linear}
    boundBone {boneName=loincloth_frnt_01_c}
    chainEnd {boneName=loincloth_frnt_03_c}
    componentAppliedLinearAccClamp=<6000 6000 6000>
    componentLinearAccScale=<1 1 1>
    componentLinearVelScale=<1 1 1>
    physicsBodyDefinitions [
     {
      sphereCollisionRadius=6
      boundBone {boneName=loincloth_frnt_01_c}
      boxExtents=<20 20 20>
      constraintSetup {angularLimitsMax=<0 0 10>, angularLimitsMin=<0 0 -50>, angularTarget=<1 0 0>}
      localJointOffset=<2 0 0>
     }
     {
      sphereCollisionRadius=5
      boundBone {boneName=loincloth_frnt_02_c}
      boxExtents=<20 20 20>
      constraintSetup {angularLimitsMax=<0 5 30>, angularLimitsMin=<0 -5 -50>, angularTarget=<0 0 0>, linearAxesMax=<0 1 0>}
      localJointOffset=<2 0 0>
     }
     {
      sphereCollisionRadius=6
      boundBone {boneName=loincloth_frnt_03_c}
      boxExtents=<20 20 20>
      constraintSetup {angularLimitsMax=<0 10 50>, angularLimitsMin=<0 -10 -50>, angularTarget=<0 0 0>, linearAxesMax=<0 1 0>}
      localJointOffset=<2 0 0>
     }
    ]
   }
   n11 = Anim Dynamics - Bone: mohawk_top(ComponentPose=@n10, Alpha=1, ExternalForce=@n2) {
    angularDampingOverride=0.9999
    angularSpringConstant=6000
    bAngularSpring=true
    bGravityOverrideInSimSpace=true
    bOverrideAngularDamping=true
    numSolverIterationsPostUpdate=4
    numSolverIterationsPreUpdate=16
    simulationSpace=RootRelative
    alphaBoolBlend {blendOption=Linear}
    boundBone {boneName=mohawk_top}
    componentAppliedLinearAccClamp=<5000 5000 5000>
    componentLinearAccScale=<1 1 1>
    componentLinearVelScale=<1 1 1>
    physicsBodyDefinitions [{
      sphereCollisionRadius=0
      boundBone {boneName=mohawk_top}
      boxExtents=<19 19 19>
      constraintSetup {angularLimitsMax=<5 10 10>, angularLimitsMin=<-5 -10 -10>, angularTarget=<1 0 0>}
      localJointOffset=<-2 0 0>
     }]
   }
   n12 = Anim Dynamics - Bone: mohawk_back(ComponentPose=@n11, Alpha=1, ExternalForce=@n2) {
    angularDampingOverride=0.9999
    angularSpringConstant=6000
    bAngularSpring=true
    bGravityOverrideInSimSpace=true
    bOverrideAngularDamping=true
    numSolverIterationsPostUpdate=4
    numSolverIterationsPreUpdate=16
    simulationSpace=RootRelative
    alphaBoolBlend {blendOption=Linear}
    boundBone {boneName=mohawk_back}
    componentAppliedLinearAccClamp=<5000 5000 5000>
    componentLinearAccScale=<1 1 1>
    componentLinearVelScale=<1 1 1>
    physicsBodyDefinitions [{
      sphereCollisionRadius=0
      boundBone {boneName=mohawk_back}
      boxExtents=<19 19 19>
      constraintSetup {angularLimitsMax=<5 10 10>, angularLimitsMin=<-5 -10 -10>, angularTarget=<1 0 0>}
      localJointOffset=<-2 0 0>
     }]
   }
   n13 = "Anim Dynamics - Chain: bangs01 -> bangs04"!(ComponentPose=@n12, Alpha=1, GravityOverride:Vector="0,0,0") -> (Pose) {
    angularDampingOverride=0.9999
    angularSpringConstant=2500
    bAngularSpring=true
    bChain=true
    bGravityOverrideInSimSpace=true
    bOverrideAngularDamping=true
    bUseGravityOverride=true
    bUsePlanarLimit=false
    bUseSphericalLimits=true
    numSolverIterationsPostUpdate=4
    numSolverIterationsPreUpdate=16
    simulationSpace=RootRelative
    alphaBoolBlend {blendOption=Linear}
    boundBone {boneName=bangs01}
    chainEnd {boneName=bangs04}
    componentAppliedLinearAccClamp=<5000 5000 5000>
    componentLinearAccScale=<0.5 0.5 0.5>
    componentLinearVelScale=<0.5 0.5 0.5>
    physicsBodyDefinitions [
     {
      sphereCollisionRadius=1
      boundBone {boneName=bangs01}
      boxExtents=<30 30 30>
      constraintSetup {angularLimitsMax=<5 20 15>, angularLimitsMin=<-5 -20 -15>, angularTarget=<0 0 0>}
      localJointOffset=<1 0 0>
     }
     {
      collisionType=CustomSphere
      sphereCollisionRadius=3
      boundBone {boneName=bangs02}
      boxExtents=<20 20 20>
      constraintSetup {angularLimitsMax=<5 20 15>, angularLimitsMin=<-5 -20 -15>, angularTarget=<1 0 0>}
      localJointOffset=<1 0 0>
     }
     {
      collisionType=CustomSphere
      sphereCollisionRadius=4
      boundBone {boneName=bangs03}
      boxExtents=<20 20 20>
      constraintSetup {angularLimitsMax=<0 20 15>, angularLimitsMin=<0 -20 -15>, angularTarget=<0 0 0>}
      localJointOffset=<1 0 0>
     }
     {
      collisionType=CustomSphere
      sphereCollisionRadius=3
      boundBone {boneName=bangs04}
      boxExtents=<15 15 15>
      constraintSetup {angularLimitsMax=<0 20 15>, angularLimitsMin=<0 -20 -15>, angularTarget=<0 0 0>}
      localJointOffset=<1 0 0>
     }
    ]
    sphericalLimits [
     {limitRadius=4, drivingBone {boneName=head}, sphereLocalOffset=<0 0 0>}
     {limitRadius=6.75, drivingBone {boneName=clavicle_l}, sphereLocalOffset=<5 -5 5>}
     {limitRadius=5.5, drivingBone {boneName=clavicle_l}, sphereLocalOffset=<5 4 2>}
    ]
   }
   n14 = "Constraint - back_ring01_dyn (PP)"!(ComponentPose=@n13, Alpha=1, ConstraintWeights_0=0.5, ConstraintWeights_1:float=0.5) -> (Pose) {
    alphaBoolBlend {blendOption=Linear}
    boneToModify {boneName=back_ring01_dyn}
    constraintSetup [{targetBone {boneName=spine_04}}, {targetBone {boneName=spine_05}}]
    constraintWeights [1,1]
   }
   n15 = "Constraint - back_ring02_dyn (PP)"!(ComponentPose=@n14, Alpha=1, ConstraintWeights_0=0.5, ConstraintWeights_1:float=0.5) -> (Pose) {
    alphaBoolBlend {blendOption=Linear}
    boneToModify {boneName=back_ring02_dyn}
    constraintSetup [{targetBone {boneName=spine_04}}, {targetBone {boneName=spine_05}}]
    constraintWeights [1,1]
   }
   n16 = Anim Dynamics - Bone: back_ring01_dyn(ComponentPose=@n15, Alpha=1, ExternalForce="0,0,0") {
    angularDampingOverride=0.99
    angularSpringConstant=3000
    bAngularSpring=true
    bGravityOverrideInSimSpace=true
    bOverrideAngularDamping=true
    bUsePlanarLimit=false
    numSolverIterationsPostUpdate=4
    numSolverIterationsPreUpdate=16
    simulationSpace=RootRelative
    alphaBoolBlend {blendOption=Linear}
    alphaScaleBiasClamp {scale=0.3}
    boundBone {boneName=back_ring01_dyn}
    componentAppliedLinearAccClamp=<5000 5000 5000>
    componentLinearAccScale=<1 1 1>
    componentLinearVelScale=<1 1 1>
    physicsBodyDefinitions [{
      sphereCollisionRadius=0
      boundBone {boneName=back_ring01_dyn}
      boxExtents=<5 5 5>
      constraintSetup {angularTargetAxis=AxisY, twistAxis=AxisY, angularLimitsMax=<60 5 5>, angularLimitsMin=<-10 -5 -5>, angularTarget=<1 0 0>}
      localJointOffset=<0 -4 0>
     }]
   }
   n17 = Anim Dynamics - Bone: back_ring02_dyn(ComponentPose=@n16, Alpha=1, ExternalForce="0,0,0") {
    angularDampingOverride=0.99
    angularSpringConstant=3000
    bAngularSpring=true
    bGravityOverrideInSimSpace=true
    bOverrideAngularDamping=true
    bUsePlanarLimit=false
    numSolverIterationsPostUpdate=4
    numSolverIterationsPreUpdate=16
    simulationSpace=RootRelative
    alphaBoolBlend {blendOption=Linear}
    alphaScaleBiasClamp {scale=0.5}
    boundBone {boneName=back_ring02_dyn}
    componentAppliedLinearAccClamp=<5000 5000 5000>
    componentLinearAccScale=<1 1 1>
    componentLinearVelScale=<1 1 1>
    physicsBodyDefinitions [{
      sphereCollisionRadius=0
      boundBone {boneName=back_ring02_dyn}
      boxExtents=<5 5 5>
      constraintSetup {angularTargetAxis=AxisY, twistAxis=AxisY, angularLimitsMax=<60 5 5>, angularLimitsMin=<-10 -5 -5>, angularTarget=<1 0 0>}
      localJointOffset=<0 -2 0>
     }]
   }
   n18 = Anim Dynamics - Bone: hip_metal_dyn(ComponentPose=@n17, Alpha=1, ExternalForce="0,0,0") {
    angularDampingOverride=0.99
    angularSpringConstant=3000
    bAngularSpring=true
    bGravityOverrideInSimSpace=true
    bOverrideAngularDamping=true
    numSolverIterationsPostUpdate=4
    numSolverIterationsPreUpdate=16
    alphaBoolBlend {blendOption=Linear}
    boundBone {boneName=hip_metal_dyn}
    componentAppliedLinearAccClamp=<5000 5000 5000>
    componentLinearAccScale=<1 1 1>
    componentLinearVelScale=<1 1 1>
    physicsBodyDefinitions [{
      sphereCollisionRadius=0
      boundBone {boneName=hip_metal_dyn}
      boxExtents=<8 8 8>
      constraintSetup {angularTargetAxis=AxisY, twistAxis=AxisY, angularLimitsMax=<10 7 5>, angularLimitsMin=<-10 -7 -1>, angularTarget=<1 0 0>}
      localJointOffset=<0 5 0>
     }]
   }
   n19 = Anim Dynamics - Bone: earring_top_dyn(ComponentPose=@n18, Alpha=0.5, ExternalForce=Get earrings_and_hair_external_force()) {
    angularDampingOverride=0.99
    angularSpringConstant=200
    bAngularSpring=true
    bGravityOverrideInSimSpace=true
    bOverrideAngularDamping=true
    numSolverIterationsPostUpdate=4
    numSolverIterationsPreUpdate=16
    alphaBoolBlend {blendOption=Linear}
    boundBone {boneName=earring_top_dyn}
    componentAppliedLinearAccClamp=<5000 5000 5000>
    componentLinearAccScale=<1 1 1>
    componentLinearVelScale=<1 1 1>
    physicsBodyDefinitions [{
      sphereCollisionRadius=0
      boundBone {boneName=earring_top_dyn}
      boxExtents=<5 5 5>
      constraintSetup {angularTargetAxis=AxisZ, twistAxis=AxisZ, angularLimitsMax=<0 30 0>, angularLimitsMin=<0 -30 0>, angularTarget=<1 0 0>}
      localJointOffset=<0 0 2>
     }]
   }
   n20 = Copy Bone - Source Bone: earring_top_dyn - Target Bone: earring_bottom_dyn!(ComponentPose:Component Space Pose Link Structure=@n19, Alpha:float=0.5, bCopyTranslation:bool=False, bCopyScale:bool=False) -> (Pose:Component Space Pose Link Structure) {
    is=title AnimGraphNode_CopyBone
    bCopyRotation=true
    alphaBoolBlend {blendOption=Linear}
    sourceBone {boneName=earring_top_dyn}
    targetBone {boneName=earring_bottom_dyn}
   }
   n21 = Get shoulder_fluff_external_force() -> (shoulder_fluff_external_force:Vector)
   n22 = Anim Dynamics - Bone: shoulder_fluff_back_dyn(ComponentPose=@n20, Alpha=1, ExternalForce=@n21) {
    angularDampingOverride=1
    angularSpringConstant=8000
    bAngularSpring=true
    bGravityOverrideInSimSpace=true
    bOverrideAngularDamping=true
    linearDampingOverride=0.7
    linearSpringConstant=100
    numSolverIterationsPostUpdate=4
    numSolverIterationsPreUpdate=16
    simulationSpace=RootRelative
    alphaBoolBlend {blendOption=Linear}
    boundBone {boneName=shoulder_fluff_back_dyn}
    componentAppliedLinearAccClamp=<5000 5000 5000>
    componentLinearAccScale=<1 1 1>
    componentLinearVelScale=<1 1 1>
    physicsBodyDefinitions [{
      sphereCollisionRadius=0
      boundBone {boneName=shoulder_fluff_back_dyn}
      boxExtents=<20 20 20>
      constraintSetup {twistAxis=AxisY, angularLimitsMax=<3 20 60>, angularLimitsMin=<-3 -20 -30>, angularTarget=<1 0 0>}
      localJointOffset=<0 -1 0>
     }]
   }
   n23 = Anim Dynamics - Bone: shoulder_fluff_dyn_l(ComponentPose=@n22, Alpha=1, ExternalForce=@n21) {
    angularDampingOverride=1
    angularSpringConstant=8000
    bAngularSpring=true
    bGravityOverrideInSimSpace=true
    bOverrideAngularDamping=true
    linearDampingOverride=0.9
    linearSpringConstant=10000
    numSolverIterationsPostUpdate=4
    numSolverIterationsPreUpdate=16
    simulationSpace=RootRelative
    alphaBoolBlend {blendOption=Linear}
    boundBone {boneName=shoulder_fluff_dyn_l}
    componentAppliedLinearAccClamp=<5000 5000 5000>
    componentLinearAccScale=<1 1 1>
    componentLinearVelScale=<1 1 1>
    physicsBodyDefinitions [{
      sphereCollisionRadius=0
      boundBone {boneName=shoulder_fluff_dyn_l}
      boxExtents=<20 20 20>
      constraintSetup {twistAxis=AxisY, angularLimitsMax=<20 20 20>, angularLimitsMin=<-20 -45 -20>, angularTarget=<1 0 0>}
      localJointOffset=<0 -1 0>
     }]
   }
   n24 = Anim Dynamics - Bone: shoulder_fluff_dyn_r(ComponentPose=@n23, Alpha=1, ExternalForce=@n21) {
    angularDampingOverride=1
    angularSpringConstant=8000
    bAngularSpring=true
    bGravityOverrideInSimSpace=true
    bOverrideAngularDamping=true
    linearDampingOverride=0.9
    linearSpringConstant=10000
    numSolverIterationsPostUpdate=4
    numSolverIterationsPreUpdate=16
    simulationSpace=RootRelative
    alphaBoolBlend {blendOption=Linear}
    boundBone {boneName=shoulder_fluff_dyn_r}
    componentAppliedLinearAccClamp=<5000 5000 5000>
    componentLinearAccScale=<1 1 1>
    componentLinearVelScale=<1 1 1>
    physicsBodyDefinitions [{
      sphereCollisionRadius=0
      boundBone {boneName=shoulder_fluff_dyn_r}
      boxExtents=<20 20 20>
      constraintSetup {twistAxis=AxisY, angularLimitsMax=<20 20 20>, angularLimitsMin=<-20 -45 -20>, angularTarget=<1 0 0>}
      localJointOffset=<0 -1 0>
     }]
   }
   Output Pose(Result=Blend Poses by bool(
     BlendPose_0=Component To Local(ComponentPose=@n1)
     BlendPose_1=Component To Local(ComponentPose=@n24)
     BlendTime_0=0
     BlendTime_1=0
     bActiveValue=Get GetPlayerListUsesCustomCharacterParts() -> (GetPlayerListUsesCustomCharacterParts:bool)
    ) {blendType=Linear}) {layerGroup=None}
   Save cached pose 'PreScale Cache'(Pose=Copy Pose From Mesh() {bUseAttachedParent=true}) {cacheName=PreScale Cache}
  ]}
 EventGraph {nodes [
   Event BlueprintUpdateAnimation()
   n1 = Get random_stream() -> (random_stream:Random Stream Structure)
   n2 = RandomFloatInRangeFromStream!(Stream:"Random Stream Structure (by ref)"=@n1, Min:float=500, Max:float=4500) -> (ReturnValue:float)
   n3 = Get bRecentlyFired() -> (bRecentlyFired:bool)
   > Set earrings_and_hair_external_force(earrings_and_hair_external_force:Vector=SelectVector(A=MakeVector(Y=0, Z=@n2), B="0, 0, 0", bPickA=@n3)) -> (Output_Get:Vector)
   n4 = float * float(A=SelectFloat(A=2800, B=0, bPickA=@n3), B=RandomFloatInRangeFromStream(Stream=@n1, Min=0.3, Max=1))
   > Set shoulder_fluff_external_force(shoulder_fluff_external_force:Vector=MakeVector(X=@n4, Y=@n4, Z=@n4)) -> (Output_Get:Vector)
   > n5 = "Gravity Override State Function (Parameter Asset) (Nativized)"(
    CharacterPartAnimInstance=Self-Reference()
    GameTimeInSeconds=GetGameTimeInSeconds()
    Parameters=&/F_MED_Commando_Viking_Parameters
   )
   > Set Gravity Override(Gravity Override=@n5)
   Event BlueprintInitializeAnimation()
   > n6 = Sequence()
   > n7 = Branch(Condition=Get bIsInFrontEnd())
   > n8 = Set simSpaceChange(simSpaceChange:AnimPhysSimSpaceType Enum=World) -> (Output_Get:AnimPhysSimSpaceType Enum)
   Set frontendAlphaScale(execute=[@n8.then, @n9.then], frontendAlphaScale:double=SelectFloat(A=0.5, B=1, bPickA=Get bIsInFrontEnd())) -> (Output_Get:double)
   n9 = Set simSpaceChange(execute=@n7.else)
   Set random_stream(execute=@n6.then_1, random_stream:Random Stream Structure=MakeRandomStream!(InitialSeed:int=0) -> (ReturnValue:Random Stream Structure)) -> (Output_Get:Random Stream Structure)
   > Set GetPlayerListUsesCustomCharacterParts(GetPlayerListUsesCustomCharacterParts:bool=GetPlaylistUsesCustomCharacterParts!() -> (ReturnValue:bool)) -> (Output_Get:bool)
  ]}
]
variables [
 earrings_and_hair_external_force {type=Vector}
 frontendAlphaScale {type=number}
 GetPlayerListUsesCustomCharacterParts {type=boolean}
 Gravity Override {type=Vector}
 random_stream {type=RandomStream, default {initialSeed=0, seed=0}}
 shoulder_fluff_external_force {type=Vector}
 simSpaceChange {type=AnimPhysSimSpaceType}
]
