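# Physics asset of simplified colliders for the female medium soldier body in the Viking Commando set, apparently for skirt simulation, going by its name.
# Single capsules cover each calf, the pelvis and spine_03.
# Both thigh bodies have no shapes, so the upper legs provide no collision.
# Constraints link calves to thighs and both thighs and spine_03 to the pelvis, all with identical swing and twist limit angles.
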PhysicsAsset /Game/Characters/Player/Female/Medium/Bodies/F_MED_Commando_Viking_01/Meshes/F_MED_Soldier_01_SkirtColliders
bodies {
 calf_l [{shapeType=Capsule, length=34.6826, radius=9.482, center=<-19.9486 -2.8284 0.1383>, rotation=<rot 0 -91.1589 -92.0231>}]
 calf_r [{shapeType=Capsule, length=34.6826, radius=9.482, center=<19.998 2.8274 0.0937>, rotation=<rot 0 88.8411 -87.0231>}]
 pelvis [{shapeType=Capsule, length=10.1327, radius=9.3783, center=<-3.425 -1.8784 0.0419>, rotation=<rot -4.1602 -89.9997 -1.4037>}]
 spine_03 [{shapeType=Capsule, length=12.1151, radius=9.6841, center=<-4.5611 1.7894 -0.1524>, rotation=<rot 6.2753 -83.7246 89.9992>}]
 thigh_l []
 thigh_r []
}
constraints [
 {bone1Name=spine_03, bone2Name=pelvis, swing1LimitDegrees=45, swing2LimitDegrees=45, twistLimitDegrees=45}
 {bone1Name=calf_l, bone2Name=thigh_l, swing1LimitDegrees=45, swing2LimitDegrees=45, twistLimitDegrees=45}
 {bone1Name=calf_r, bone2Name=thigh_r, swing1LimitDegrees=45, swing2LimitDegrees=45, twistLimitDegrees=45}
 {bone1Name=thigh_l, bone2Name=pelvis, swing1LimitDegrees=45, swing2LimitDegrees=45, twistLimitDegrees=45}
 {bone1Name=thigh_r, bone2Name=pelvis, swing1LimitDegrees=45, swing2LimitDegrees=45, twistLimitDegrees=45}
]
